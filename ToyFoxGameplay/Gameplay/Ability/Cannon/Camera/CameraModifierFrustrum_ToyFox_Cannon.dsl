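; Camera modifier for ToyFox's Cannon ability that overrides the target camera rotation query.
; GetTargetCameraRotation fetches the camera owner and reads its current camera rotation.
; It keeps only the yaw and zeroes pitch and roll, so the target rotation stays level.
; When the camera owner lookup fails, it returns a zero rotation.
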(fn GetTargetCameraRotation
  (bind _cameraowner (Variables|CameraModifier|GetCameraOwner)
    (:then
      (bind _returnvalue (Camera|GetCameraRotation _cameraowner))
      (bind (_returnvalue_roll _returnvalue_pitch _returnvalue_yaw) (Math|Rotator|BreakRotator _returnvalue))
      (return 0.0 0.0 _returnvalue_yaw))
    (:else
      (return "0, 0, 0"))))
